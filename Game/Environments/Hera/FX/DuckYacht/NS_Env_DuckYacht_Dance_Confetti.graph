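# Niagara GPU effect that sprays a continuous stream of multicolored confetti for the Duck Yacht dance environment in Hera.
# Particles spawn in shapes offset to five points along a Y-axis row, then launch upward and along +X.
# Gravity, curl noise, wind, curved drag and a speed cap make them flutter before shrinking and fading.
# Mesh and sprite renderers are both present, with sprites aligned to the spinning mesh orientation and sub-UV animated.
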NiagaraSystem /Game/Environments/Hera/FX/DuckYacht/NS_Env_DuckYacht_Dance_Confetti
dependencies {
 usedDataInterfaces [
  &NiagaraDataInterfaceColorCurve
  &NiagaraDataInterfaceArrayFloat3
  &NiagaraDataInterfaceLandscape
  &NiagaraDataInterfaceCurve
 ]
 usedDynamicInputs []
 usedModules [
  &SystemState
  &self:SystemSpawnScript.SystemScriptSource.SystemScriptGraph.NiagaraNodeAssignment_0.SetVariables_#1
 ]
}
emitters {Confetti {
  simTarget=GPUComputeSim
  emitterUpdateScript [
   &EmitterState()
   &SpawnRate(SpawnRate=1500, Use Spawn Probability=false)
  ]
  particleSpawnScript [
   /SetVariables(Particles.SpawnGroup=&RandomRangeInt(Minimum=0, Maximum=3)) {enabled=false, isSetParameters=true}
   /SetVariables(
    Particles.HeightOffset:NiagaraFloat=$System.System_HeightOffset
    Particles.UporDown:NiagaraFloat=&UniformAOrBFloat(
     A=&RandomRangeFloat(Minimum=-200, Maximum=200)
     B=&RandomRangeFloat(Minimum=0, Maximum=0)
     Distribution Weight=0.8
    )
   ) {isSetParameters=true}
   &InitializeParticle(
    Color=&ColorFromCurve(
     ColorCurve=DataInterface() {propertyValues {
       AlphaCurve {keys [<at 0 1 RCIM_Cubic>,<at 1 1 RCIM_Cubic>]}
       BlueCurve {keys [
         {arriveTangent=-1, interpMode=RCIM_Cubic, time=0.1211, value=0.1095}
         {arriveTangent=-0.1153, interpMode=RCIM_Cubic, leaveTangent=-0.1153, time=0.5, value=0.1829}
         {interpMode=RCIM_Cubic, leaveTangent=0.613, time=0.75, value=0.0369}
        ]}
       GreenCurve {keys [
         {arriveTangent=1, interpMode=RCIM_Cubic, time=0.1211, value=0.6298}
         {arriveTangent=-0.4666, interpMode=RCIM_Cubic, leaveTangent=-0.4666, time=0.5, value=0.6204}
         {interpMode=RCIM_Cubic, leaveTangent=-0.7408, time=0.75, value=0.3363}
        ]}
       RedCurve {keys [
         {arriveTangent=-0.7434, interpMode=RCIM_Cubic, time=0.1211, value=1}
         {arriveTangent=-1.5635, interpMode=RCIM_Cubic, leaveTangent=-1.5635, tangentMode=RCTM_User, time=0.5, value=1}
         {interpMode=RCIM_Cubic, leaveTangent=-1, time=0.75, value=1}
        ]}
      }}
     CurveIndex=&RandomRangeFloat(Minimum=0, Maximum=1)
    )
    Lifetime Max=5
    Lifetime Min=3
    Lifetime Mode=::NewEnumerator1
    Mesh Scale Mode=::NewEnumerator1
    Mesh Uniform Scale Max=25
    Mesh Uniform Scale Min=5
    Sprite Rotation Mode=::NewEnumerator2
    Sprite Size Mode=::NewEnumerator1
    UsePositionOffset=false
    Write Material Random=true
   ) {module=InitializeParticle001}
   &AddVelocity(Velocity=&RandomRangeVector(Minimum=<1500 0 1500>, Maximum=<2500 1 5000>))
   &ShapeLocation(
    Hemisphere Distribution=<1 0.8>
    MaskSpawnGroup=false
    Offset=&SelectVectorFromArray(
     Array Sampling Mode=::NewEnumerator0
     Direct Array Index=?
     Vector Selection Array=DataInterface() {propertyValues {FloatData [<0 -400 0>, <1 -200 1>, <0 0 0>, <0 198 0>, <0 400 0>]}}
     Random Seed=0
    )
    Offset Mode=::NewEnumerator0
    Override Seed=false
    UseOwnerScale=true
   )
   /NiagaraFN/Modules/Spawn/FN_SpawnParticlesOnTerrainNearPlayer(
    HeightOffset:NiagaraFloat=&RandomRangeFloat(Minimum=0, Maximum=1500)
    Player Position:NiagaraPosition=$Particles.Position
   ) {enabled=false}
   /Niagara/Modules/Update/Camera/MaintainInCameraParticleScale(Render Target Width=2)
   &CameraOffset()
   &SubUVAnimation(End Frame=36, SubUV Animation Mode:ENiagara_SubUVLookupMode=::NewEnumerator2) {module=SubUVAnimation001}
   &Color(Color=&ColorFromCurve(
     ColorCurve=DataInterface() {propertyValues {
       AlphaCurve {keys [<at 0 1>,<at 1 1>]}
       BlueCurve {keys [<at 0 0.8714>,<at 0.2389 0.8714>,<at 0.2977 0.7011>,<at 0.4603 0.7011>,<at 0.5416 0.1624>,<at 1 0.1624>]}
       GreenCurve {keys [<at 0 0.5149>,<at 0.2389 0.5149>,<at 0.2977 0.063>,<at 0.4603 0.063>,<at 0.5416 0.9583>,<at 1 0.9583>]}
       RedCurve {keys [<at 0 0.0144>,<at 0.2389 0.0144>,<at 0.2977 1>,<at 0.4603 1>,<at 0.5416 0>,<at 1 0>]}
      }}
     CurveIndex=$Particles.MaterialRandom
    )) {enabled=false}
   &InitialMeshOrientation(Use Rotation Vector=false)
  ]
  particleUpdateScript [
   &ParticleState()
   &UpdateMeshOrientation(Rotation Rate=2)
   &ScaleMeshSize(Scale Factor=&VectorFromFloat(Value=&FloatFromCurve(FloatCurve=DataInterface() {propertyValues {Curve {keys [<at 0.8 1>,<at 1 0>]}}}, Scale Curve=0.25)))
   &GravityForce(Gravity=<0 0 -100>)
   &CurlNoiseForce(Noise Frequency=5, Noise Strength=250, Pan Noise=true, Pan Noise Field=<0 0 1>)
   &WindForce(
    Friction Mode:ENiagaraWindFrictionMode=::NewEnumerator0
    Offset By Time:Vector3f=<0 0 1>
    Scale Min / Max:Vector2f=<2 1>
    Speed Min / Max:Vector2f=<0 800>
    Surface Alignment Mode:ENiagaraWindCombingMode=::NewEnumerator0
    Turbulence Scale Mode:ENiagaraWindTurbulenceContributionMode=::NewEnumerator1
    Wind Speed=<25 0 0>
   )
   &Drag(
    Drag=&FloatFromCurve(FloatCurve=DataInterface() {propertyValues {Curve {keys [<at 0 0>,<at 0.2598 0.8029>,<at 1 0.5>]}}})
    Rotational Drag=0.02
    Use Linear Drag=true
    Use Rotational Drag=true
   ) {module=Drag001}
   &MeshRotationForce(Rotation=<5 0 0>)
   &SolveForcesAndVelocity(Clamp Velocity=true, Limit Acceleration=false, Speed Limit=500)
   &SpriteRotationRate(Rotation Rate=&FloatFromCurve(
     FloatCurve=DataInterface() {propertyValues {Curve {keys [<at 0 1>,<at 1 0>]}}}
     Scale Curve=&AngleConversion(Angle:NiagaraFloat=0, Angle Input:ENiagara_AngleInput=::NewEnumerator0, Angle Output:ENiagara_AngleInput=::NewEnumerator1)
    ))
   &SubUVAnimation(End Frame=16, SubUV Animation Mode:ENiagara_SubUVLookupMode=::NewEnumerator1, SubUV Loop Count=2)
   &DynamicMaterialParameters(
    Index 0 Param 1=0.2
    Module.Index0Float0Write=true
    Module.Index0Float1Write=true
    Module.Index0Float2Write=true
    Module.Index0Float3Write=true
   )
   &AlignSpriteToMeshOrientation(SetAlignment=true, SetFacing=true)
   &ScaleColor(
    Scale Alpha=&FloatFromCurve(FloatCurve=DataInterface() {propertyValues {Curve {keys [
         {interpMode=RCIM_Cubic, tangentMode=RCTM_User, time=0, value=0}
         {interpMode=RCIM_Cubic, tangentMode=RCTM_User, time=0.2, value=1}
         {interpMode=RCIM_Cubic, tangentMode=RCTM_User, time=0.9, value=1}
         {tangentMode=RCTM_User, time=1, value=0}
        ]}}})
    Scale RGB=&MakeVectorFromLinearColorRGB(Color=&HueShiftLinearColor(
      Hue=0
      Saturation=0.5
      Value=2
      Linear Color=&MultiplyLinearColor(Linear Color=$Particles.Initial.Color, Scale Factor=&MakeVector4_FromFloat(Float=1))
     ))
    ScaleA=true
    ScaleRGB=true
   )
  ]
  rendererClasses [&NiagaraSpriteRendererProperties, &NiagaraMeshRendererProperties]
  renderers [
   {rendererClass=&NiagaraSpriteRendererProperties}
   {rendererClass=&NiagaraMeshRendererProperties, rendererIndex=1}
  ]
 }}
summary {
 systemName=NS_Env_DuckYacht_Dance_Confetti
 emitters [{
   emitterName=Confetti
   simTarget=GPUComputeSim
   rendererClasses [&NiagaraSpriteRendererProperties, &NiagaraMeshRendererProperties]
  }]
 userVariables []
}
